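; GECAR_IsHealingEnabled is a gameplay effect custom application requirement in the Mash playlist that lets an effect apply only when healing is enabled.
; Its CanApplyGameplayEffect override evaluates the HealAmount scalable float variable and returns true when the value is greater than zero.
; The incoming effect, spec and ability system component are not inspected, so the gate depends solely on that configured amount.
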(fn CanApplyGameplayEffect
  (params (GameplayEffect :type Gameplay%20Effect%20Object%20Reference) (Spec :type Gameplay%20Effect%20Spec%20Structure%20%28by%20ref%29) (ASC :type Ability%20System%20Component%20Object%20Reference))
  (return (> (ScalableFloat|GetValueatLevel (Variables|Default|GetHealAmount)) 0)))
